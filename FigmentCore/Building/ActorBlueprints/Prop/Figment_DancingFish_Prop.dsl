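; Singing bass fish prop that dances to a music cue when players emote inside its trigger volume.
; The authority tracks overlapping pawns and runs a looping timer while any remain, checking for emotes and broadcasting a DanceWithFish custom verb tagged per instance.
; The bPlayerDancing RepNotify triggers the fish montage and music, guarded against restarts.
; When the montage completes with nobody dancing, the music stops.
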(event Custom|DanceFishyDance
  (bind _singingbass_music_cue (Variables|Default|GetSingingBass_Music_Cue))
  ;[0 0 0] Make Fish Dance
  (Utilities|FlowControl|DoOnce false
    (:Completed
      (Animation|Montage|PlayMontage (Variables|Default|GetSK_Prop_Ranger_Fish_Budgeted) (Variables|Default|GetMontage)
        (:then
          _)
        (:OnCompleted
          (if (Variables|Default|GetbPlayerDancing)
            _)
          (Audio|Components|Audio|Stop _singingbass_music_cue)
          (Variables|Default|SetMusicAlreadyStarted false)))
      (if (not (Variables|Default|GetMusicAlreadyStarted))
        (Audio|Components|Audio|Play _singingbass_music_cue 0.0)
        (Variables|Default|SetMusicAlreadyStarted true)))))

(event OnComponentBeginOverlap(DanceTriggerOverlap)
  (params (OverlappedComponent :type Primitive%20Component%20Object%20Reference) (OtherActor :type Actor%20Object%20Reference) (OtherComp :type Primitive%20Component%20Object%20Reference) (OtherBodyIndex :type Integer) (bFromSweep :type Boolean) (SweepResult :type Hit%20Result%20Structure%20%28by%20ref%29))
  ;[0 0 0] On overlap dance volume
  ;; When any player overlaps, track them in an array and listen for when they dance
  (Networking|SwitchHasAuthority
    (:Authority
      (bind _asfort_player_pawn (Utilities|Casting|CastToFortPlayerPawn OtherActor))
      (Utilities|Array|AddUnique (Variables|Default|GetPlayersInVolume) _asfort_player_pawn)
      (if (not (Utilities|Time|IsValidTimerHandle (Variables|Default|GetDanceTimerHandle)))
        (bind _returnvalue (Utilities|Time|SetTimerbyEvent (delegate Custom|CheckForPlayersDancing) 1.0 true true 0.5 0.1))
        (Variables|Default|SetDanceTimerHandle _returnvalue)))))

(event Custom|CheckForPlayersDancing
  ;[0 0 0] On overlap dance volume
  ;; When any player overlaps, track them in an array and listen for when they dance
  (CallFunction|MakeFishDanceforEmotingPlayers))

(event OnComponentEndOverlap(DanceTriggerOverlap)
  (params (OverlappedComponent :type Primitive%20Component%20Object%20Reference) (OtherActor :type Actor%20Object%20Reference) (OtherComp :type Primitive%20Component%20Object%20Reference) (OtherBodyIndex :type Integer))
  ;[0 0 0] On overlap dance volume
  ;; When a player leaves, remove them from the tracking array. If there are no players left in the volume, stop the timer
  (Networking|SwitchHasAuthority
    (:Authority
      (bind _asfort_player_pawn (Utilities|Casting|CastToFortPlayerPawn OtherActor))
      (if (Utilities|Array|ContainsItem (Variables|Default|GetPlayersInVolume) _asfort_player_pawn)
        (Utilities|Array|RemoveItem (Variables|Default|GetPlayersInVolume) _asfort_player_pawn)
        (if (Utilities|Array|IsEmpty (Variables|Default|GetPlayersInVolume))
          (Utilities|Time|ClearandInvalidateTimerbyHandle (Variables|Default|GetDanceTimerHandle))
          (Variables|Default|SetbPlayerDancing false))))))

(event EventEndPlay
  (params (EndPlayReason :type EEndPlayReason%20Enum :default "Destroyed"))
  (bind _dancetimerhandle (Variables|Default|GetDanceTimerHandle))
  ;[0 0 0] End play - Clean up the timer and stored players in the map variable
  (|CallParentFunction EndPlayReason)
  (Networking|SwitchHasAuthority
    (:Authority
      (if (Utilities|Time|IsValidTimerHandle _dancetimerhandle)
        (Utilities|Time|ClearandInvalidateTimerbyHandle _dancetimerhandle)
        (Utilities|Array|Clear (Variables|Default|GetPlayersInVolume))))))

(fn MakeFishDanceForEmotingPlayers
  ;[0 0 0] If any players in the trigger volume are emoting, make the fish dance
  (LocalVariables|SetAnyPlayersDancing false)
  (bind _array_element (Utilities|Array|ForEachLoopwithBreak (Variables|Default|GetPlayersInVolume))
    (:LoopBody
      (if (Emote|IsPlayingEmote (Fortnite|GetFortPlayerControllerfromActor _array_element))
        (Variables|Default|SetbPlayerDancing true)
        (LocalVariables|SetAnyPlayersDancing true)
        (bind _result (CustomVerb|AddTagsforSubject (CustomVerb|MakeCustomVerbMessage _array_element "(TagName=\"CustomVerb.DanceWithFish\")") "FishPropInstance" (Variables|FillOutPerInstance|GetFishPropInstanceTags)))
        (CustomVerb|BroadcastCustomVerbMessage _result)))
    (:Completed
      (if (not (LocalVariables|GetAnyPlayersDancing))
        (Variables|Default|SetbPlayerDancing false)))))

(fn OnRep_bPlayerDancing
  (if (Variables|Default|GetbPlayerDancing)
    (CallFunction|DanceFishyDance)))

(fn ConstructionScript)
